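# Unlit alpha-composite debug material for a seamless-tiling blend in the water caustics folder, operating on raw UVs without sampling textures.
# Half-offset, wrapped UV copies blend through soft center bands on each axis to cover the cross seams.
# A custom cubic edge mask blends the original UVs back in away from the borders.
# An S-curved FrameAlpha scales the result; no material output connection is shown.
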Material /Water/Caustics/Materials/Debug/DebugTilingTexture
graph {
 properties {
  bAllowNegativeEmissiveColor=true
  blendMode=BLEND_AlphaComposite
  pixelDepthOffsetMode=PDOM_Legacy
  shadingModel=MSM_Unlit
  usedShadingModels=Unlit
 }
 nodes [
  n1 = TextureCoordinate()
  n2 = Constant2Vector() {g=0.5}
  n3 = MaterialFunctionCall(Float2=TextureCoordinate()) {materialFunction=&BreakOutFloat2Components}
  n4 = Constant() {r=0.5}
  n5 = ScalarParameter() {defaultValue=0.1, parameterName=CrossFixWidth}
  n6 = LinearInterpolate(A=Frac(None=Add(A=@n1) {constB=0.5}), B=Frac(None=Add(A=@n1, B=@n2)), Alpha=SphereMask(A=@n3.R, B=@n4, Radius=@n5) {hardnessPercent=0})
  n7 = Frac(None=Add(A=@n1, B=Constant2Vector() {r=0.5}))
  n8 = SphereMask(A=@n3.G, B=@n4, Radius=@n5) {hardnessPercent=0}
  MaterialExpressionScalarParameter_23 = ScalarParameter() {defaultValue=0.25, parameterName=sm}
  MaterialExpressionTextureCoordinate_10 = TextureCoordinate()
  n9 = Custom!(SplineMargin=@MaterialExpressionScalarParameter_23, UV=@MaterialExpressionTextureCoordinate_10) {
   code="//input float HeightScale;\r\n//input float SplineMargin;\r\n\r\nfloat SplineParameter1 = (0.5f - SplineMargin);\r\nfloat SplineParameter2 = (1.0f/SplineMargin);\r\n\r\n\r\n    // Re-scale [0..0.1] and [0.9..1] to [0..1]\r\n    float2 T        = max(abs(UV - 0.5f) - SplineParameter1, 0.0f) * SplineParameter2;\r\n    // Cubic interpolation\r\n    float2 F        = 2*T*T*T - 3*T*T + 1;\r\n    // Bilinear filter from 2D to 1D factor (making it zero at the edges)\r\n    float Scale        =  F.x * F.y;\r\n\r\nreturn Scale;"
   description=Soft Edge Mask
   outputType=CMOT_Float1
   inputs [
    {inputName=SplineMargin, input {expression=&self:MaterialExpressionScalarParameter_23}}
    {inputName=UV, input {expression=&self:MaterialExpressionTextureCoordinate_10}}
   ]
  }
  Multiply(
   A=LinearInterpolate(A=LinearInterpolate(A=@n6, B=@n7, Alpha=@n8), B=@n1, Alpha=@n9)
   B=MaterialFunctionCall(In=ScalarParameter() {parameterName=FrameAlpha}, Power=ScalarParameter() {defaultValue=3, parameterName=Frame Blend Contrast}) {materialFunction=&SCurve}.Result
  )
 ]
}
