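# Unlit debug material for the water caustics system that inspects the time controls in the MPC_Water parameter collection.
# It feeds the FreezeTime parameter into the DebugScalarValues material function for numeric display.
# A separate lerp blends engine Time toward ManualTime using FreezeTime as alpha, but its result feeds nothing.
# The graph shows no connection from either branch to a material output.
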Material /Water/Caustics/Materials/Debug/DebugWaterTime
graph {
 properties {pixelDepthOffsetMode=PDOM_Legacy, shadingModel=MSM_Unlit, usedShadingModels=Unlit}
 nodes [
  n1 = CollectionParameter() {collection=/Water/Materials/MPC/MPC_Water, parameterName=FreezeTime}
  LinearInterpolate(A=Time(), B=CollectionParameter() {collection=&MPC_Water, parameterName=ManualTime}, Alpha=@n1)
  MaterialFunctionCall(Number=@n1) {materialFunction=&DebugScalarValues}
 ]
}
