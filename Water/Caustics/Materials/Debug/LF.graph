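# Debug light function material in the Water Caustics folder, using the Unlit shading model.
# Its only chain feeds the red and green channels of the Light Vector as UVs into a sample of the engine's AICON-Green texture.
# The texture sample is not wired to any material output, so the graph does not drive the light function's result.
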Material /Water/Caustics/Materials/Debug/LF
graph {
 properties {materialDomain=MD_LightFunction, pixelDepthOffsetMode=PDOM_Legacy, shadingModel=MSM_Unlit, usedShadingModels=Unlit}
 nodes [TextureSample(UVs=ComponentMask(None=LightVector()) {g=true, r=true}) {texture=/Engine/EngineResources/AICON-Green}]
}
